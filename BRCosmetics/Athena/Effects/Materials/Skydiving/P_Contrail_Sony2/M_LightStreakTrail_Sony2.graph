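# Unlit translucent, two-sided particle material for the light streak trail of the Sony2 skydiving contrail.
# A sine wobble, phased by time and actor height and tuned by dynamic parameters, offsets both the tapered streak profile and panning plasma and caustic noise.
# Curve atlas channels shape the core, edges and length fade.
# Inside/outside colors and particle color drive time-of-day-scaled emissive; particle alpha and camera depth fade modulate opacity.
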Material /BRCosmetics/Athena/Effects/Materials/Skydiving/P_Contrail_Sony2/M_LightStreakTrail_Sony2
graph {
 properties {
  bUsedWithNiagaraSprites=true
  bUsedWithParticleSprites=true
  blendMode=BLEND_Translucent
  pixelDepthOffsetMode=PDOM_Legacy
  shadingModel=MSM_Unlit
  twoSided=true
  usedShadingModels=Unlit
 }
 nodes [
  n1 = Constant() {r=2}
  n2 = Constant() {r=1}
  n3 = DynamicParameter() {paramNames [ReverseU, "SinMultiple(0.4Def)", SinSpeed, Param4]}
  n4 = TextureCoordinate()
  n5 = ComponentMask(None=@n4) {r=true}
  n6 = AppendVector(A=Abs(None=Subtract(A=@n3.ReverseU, B=@n5)), B=ComponentMask(None=@n4) {g=true})
  n7 = CurveAtlasRowParameter(CurveTime=OneMinus(None=ComponentMask(None=@n6) {g=true})) {atlas=&AthenaCurveAtlas01, curve=/Game/Athena/LinearColorCurves/Curve3_SkydiveSony2, defaultValue=93}
  n8 = Time()
  n9 = ActorPositionWS()
  n10 = Multiply(A=OneMinus(None=ComponentMask(None=@n9) {b=true})) {constB=0.005}
  n11 = Constant2Vector() {g=0.2}
  n12 = DynamicParameter() {paramNames [ReverseU, "SinMultiple(0.4Def)", SinSpeed, Param4]}
  n13 = ComponentMask(None=Panner(Coordinate=@n6, Time=Add(A=@n8, B=@n10), Speed=Multiply(A=@n11, B=@n12.SinSpeed)) {speedY=0.2}) {g=true}
  n14 = Multiply(A=@n7.B, B=Multiply(A=Sine(None=@n13), B=@n12."SinMultiple(0.4Def)") {constB=0.4})
  n15 = AppendVector(A=@n14, B=Constant())
  n16 = Multiply(A=@n6, B=Constant2Vector() {g=0.1, r=2})
  n17 = TextureSample(UVs=Add(A=Multiply(A=@n15) {constB=2}, B=Panner(Coordinate=@n16) {speedX=0.2, speedY=0.4})) {texture=&T_ShieldCaustics}
  n18 = ParticleColor()
  n19 = Multiply(A=Clamp(None=LinearInterpolate(Alpha=@n17.B) {constA=-2}), B=@n18.RGB)
  n20 = Multiply(A=@n6, B=Constant2Vector() {g=1.5, r=2})
  n21 = Add(A=Frac(None=Panner(Coordinate=@n20) {speedY=1}), B=Multiply(A=@n15) {constB=2})
  n22 = Constant2Vector() {g=1.2, r=2}
  n23 = Add(A=Panner(Coordinate=Multiply(A=@n6, B=@n22)) {speedX=0.1, speedY=0.8}, B=Multiply(A=@n15) {constB=3})
  n24 = Add(
   A=Multiply(
    A=TextureSample(UVs=@n21) {texture=/Game/Effects/Fort_Effects/Textures/T_FX_SpringerPlasma}.R
    B=TextureSample(UVs=@n23) {texture=&T_ShieldCaustics}.R
   )
   B=LinearInterpolate(Alpha=@n17.B) {constA=-0.1, constB=0.2}
  )
  n25 = ScalarParameter() {defaultValue=0.8, parameterName=ShapeShrink}
  n26 = ScalarParameter() {defaultValue=0.35, parameterName=TaperAmount}
  n27 = ComponentMask(None=@n6) {r=true}
  n28 = Multiply(A=OneMinus(None=ComponentMask(None=@n6) {g=true}), B=LinearInterpolate(A=@n26, B=Multiply(A=@n26) {constB=-1}, Alpha=@n27) {constB=0})
  n29 = Add(A=LinearInterpolate(A=Multiply(A=@n25) {constB=-1}, B=Add(A=@n25), Alpha=Add(A=@n28, B=@n27)) {constA=-0.2, constB=1.2}, B=@n14)
  n30 = CurveAtlasRowParameter(CurveTime=@n29) {atlas=&AthenaCurveAtlas01, curve=&Curve3_SkydiveSony2, defaultValue=93}
  n31 = CurveAtlasRowParameter(CurveTime=@n29) {atlas=&AthenaCurveAtlas01, curve=&Curve3_SkydiveSony2, defaultValue=93}
  n32 = ComponentMask(None=TextureCoordinate()) {g=true}
  n33 = Multiply(A=Add(A=Multiply(A=@n31.G, B=@n32), B=Multiply(A=@n30.R, B=@n24)))
  n34 = VectorParameter() {parameterName=DarkOutsideColor, defaultValue=<rgba 0.045 0 0.7344 1>}
  n35 = VectorParameter() {parameterName=DarkInsideColor, defaultValue=<rgba 1 1 1 1>}
  n36 = LinearInterpolate(Alpha=@n17.B) {constB=20}
  n37 = Multiply(
   A=LinearInterpolate(
    A=VectorParameter() {parameterName=OutsideColor, defaultValue=<rgba 0 0.087 1 1>}.RGB
    B=VectorParameter() {parameterName=InsideColor, defaultValue=<rgba 1 1 1 1>}.RGB
    Alpha=Clamp(None=@n33)
   )
   B=Clamp(None=LinearInterpolate(A=@n34.RGB, B=@n35.RGB, Alpha=@n36))
  )
  MaterialFunctionCall(Emissive=Multiply(
    A=MaterialFunctionCall(Value_Morning=@n1, Value_Day=@n1, Value_Evening=@n1, Value_Night=@n2) {materialFunction=&Scalar_Time_of_Day}.Result
    B=Add(A=@n19, B=@n37)
   )) {materialFunction=&MF_FX_EmissiveAdjustments}
  n38 = Add(A=@n31.G, B=@n30.R)
  n39 = CurveAtlasRowParameter(CurveTime=ComponentMask(None=TextureCoordinate()) {g=true}) {atlas=&AthenaCurveAtlas01, curve=&Curve3_SkydiveSony2, defaultValue=93}
  Multiply(
   A=Multiply(A=@n18.A, B=Multiply(A=@n38, B=@n39.A))
   B=MaterialFunctionCall(Fade Length=Constant() {r=30}) {materialFunction=&CameraDepthFade}.Result
  )
 ]
}
