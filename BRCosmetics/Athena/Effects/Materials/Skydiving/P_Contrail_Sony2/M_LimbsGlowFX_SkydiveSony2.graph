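# Additive, unlit particle material for the glowing limb trails of the Sony2 skydive contrail.
# A base color, scaled by time of day, drives emissive.
# The glow mask is procedural: tapered UVs get a panning sine wobble, phase-shifted by actor height and tuned by dynamic parameters.
# Curve atlas rows shape wobble, cross-section and length falloff, then particle alpha and depth fade soften the result.
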Material /BRCosmetics/Athena/Effects/Materials/Skydiving/P_Contrail_Sony2/M_LimbsGlowFX_SkydiveSony2
graph {
 properties {
  bUsedWithNiagaraSprites=true
  bUsedWithParticleSprites=true
  blendMode=BLEND_Additive
  pixelDepthOffsetMode=PDOM_Legacy
  shadingModel=MSM_Unlit
  twoSided=true
  usedShadingModels=Unlit
 }
 nodes [
  n1 = Constant() {r=2}
  n2 = Constant() {r=1}
  MaterialFunctionCall(Emissive=Multiply(
    A=MaterialFunctionCall(Value_Morning=@n1, Value_Day=@n1, Value_Evening=@n1, Value_Night=@n2) {materialFunction=&Scalar_Time_of_Day}.Result
    B=VectorParameter() {parameterName=BaseColor, defaultValue=<rgba 0 0.137 1 1>}.RGB
   )) {materialFunction=&MF_FX_EmissiveAdjustments}
  n3 = ScalarParameter() {parameterName=ShapeShrink}
  n4 = TextureCoordinate()
  n5 = ScalarParameter() {defaultValue=-0.2, parameterName=TaperAmount}
  n6 = StaticComponentMaskParameter(None=@n4) {defaultR=true, parameterName=Param_1}
  n7 = Multiply(A=OneMinus(None=ComponentMask(None=@n4) {g=true}), B=LinearInterpolate(A=@n5, B=Multiply(A=@n5) {constB=-1}, Alpha=@n6) {constB=0})
  n8 = CurveAtlasRowParameter(CurveTime=OneMinus(None=ComponentMask(None=@n4) {g=true})) {atlas=&AthenaCurveAtlas01, curve=/Game/Athena/LinearColorCurves/Curve3_SkydiveSony2, defaultValue=93}
  n9 = Add(A=@n4, B=Multiply(A=DynamicParameter() {paramNames [ReverseU, "SinMultiple(0.4Def)", SinSpeed, Param4]}.ReverseU) {constB=0.5})
  n10 = ComponentMask(None=ActorPositionWS()) {b=true}
  n11 = Add(A=Time(), B=Multiply(A=OneMinus(None=@n10)) {constB=0.005})
  n12 = DynamicParameter() {paramNames [ReverseU, "SinMultiple(0.4Def)", SinSpeed, Param4]}
  n13 = Multiply(A=Constant2Vector() {g=0.2}, B=@n12.SinSpeed)
  n14 = Sine(None=ComponentMask(None=Panner(Coordinate=@n9, Time=@n11, Speed=@n13) {speedY=0.2}) {g=true})
  n15 = Add(
   A=LinearInterpolate(A=Multiply(A=@n3) {constB=-1}, B=Add(A=@n3), Alpha=Add(A=@n7, B=@n6)) {constA=-0.2, constB=1.2}
   B=Multiply(A=@n8.B, B=Multiply(A=@n14, B=@n12."SinMultiple(0.4Def)") {constB=0.4})
  )
  n16 = Multiply(A=Abs(None=Subtract(A=@n15) {constB=0.5})) {constB=2}
  n17 = CurveAtlasRowParameter(CurveTime=StaticComponentMaskParameter(None=OneMinus(None=@n16)) {defaultR=true}) {atlas=&AthenaCurveAtlas01, curve=/Game/Athena/LinearColorCurves/Curve_Skydive_Sony2, defaultValue=90}
  n18 = TextureCoordinate()
  n19 = Sine(None=Multiply(A=OneMinus(None=@n18)) {constB=0.25})
  n20 = Multiply(A=@n17.B, B=Power(Base=ComponentMask(None=@n19) {g=true}))
  n21 = CurveAtlasRowParameter(CurveTime=ComponentMask(None=@n18) {g=true}) {atlas=&AthenaCurveAtlas01, curve=&Curve_Skydive_Sony2, defaultValue=90}
  Multiply(A=DepthFade(FadeDistance=Constant() {r=10}), B=Multiply(A=ParticleColor().A, B=Multiply(A=@n20, B=@n21.A)))
 ]
}
